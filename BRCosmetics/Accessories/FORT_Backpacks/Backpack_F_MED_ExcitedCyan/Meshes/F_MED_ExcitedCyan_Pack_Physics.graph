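# Physics asset for the ExcitedCyan backpack cosmetic, defining collision and joint limits for its dangling zipper pulls and keychain.
# The pack body uses three capsules, while two zip bones and the first keychain bone are capsules constrained to it.
# A boxed second keychain bone hangs from the first, forming a two-link chain.
# Each joint has its own swing and twist limits to restrict the motion.
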PhysicsAsset /BRCosmetics/Accessories/FORT_Backpacks/Backpack_F_MED_ExcitedCyan/Meshes/F_MED_ExcitedCyan_Pack_Physics
bodies {
 keychain_01_dyn [{
   shapeType=Capsule
   length=1.3157
   radius=1.1309
   shapeName=keychain_01_dyn_capsule
   center=<-0.0409 -0.0237 -1.0571>
   rotation=<rot 6.1439 -90.4248 -3.4997>
  }]
 keychain_02_dyn [{
   shapeType=Box
   extentX=4.5868
   extentY=1.1026
   extentZ=4.5309
   shapeName=keychain_02_dyn_box
   center=<-0.2999 -1.2658 -2.8742>
   rotation=<rot 0 0 17.5131>
  }]
 pack [
  {shapeType=Capsule, length=10.0946, radius=4.5542, shapeName=pack_capsule}
  {shapeType=Capsule, length=11.0559, radius=3.2546, shapeName=pack_capsule, center=<-8.6973 -7.8773 -5.4768>, rotation=<rot -3.6568 0 -0.0004>}
  {
   shapeType=Capsule
   length=3.5361
   radius=3.9266
   shapeName=pack_capsule
   center=<-11.371 -6.673 -6.8285>
   rotation=<rot -88.6989 173.4242 -177.0115>
  }
 ]
 zip_01_dyn [{
   shapeType=Capsule
   length=1.3407
   radius=0.505
   shapeName=zip_01_dyn_capsule
   center=<0.0005 -0.0019 -0.7493>
   rotation=<rot -0.1521 -90 1.0824>
  }]
 zip_03_dyn [{
   shapeType=Capsule
   length=1.2378
   radius=0.9027
   shapeName=zip_03_dyn_capsule
   center=<0.0017 -0.0009 -0.7916>
   rotation=<rot 0.7667 -90 0.1174>
  }]
}
constraints [
 {bone1Name=zip_03_dyn, bone2Name=pack, swing1LimitDegrees=2, swing2LimitDegrees=2, twistLimitDegrees=50}
 {bone1Name=zip_01_dyn, bone2Name=pack, swing1LimitDegrees=3, swing2LimitDegrees=29.36, twistLimitDegrees=2}
 {bone1Name=keychain_01_dyn, bone2Name=pack, swing1LimitDegrees=14, swing2LimitDegrees=20, twistLimitDegrees=45}
 {bone1Name=keychain_02_dyn, bone2Name=keychain_01_dyn, swing1LimitDegrees=15, swing2LimitDegrees=10, twistLimitDegrees=45}
]
