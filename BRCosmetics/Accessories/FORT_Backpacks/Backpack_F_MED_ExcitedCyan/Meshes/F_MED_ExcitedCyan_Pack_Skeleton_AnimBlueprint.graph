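# Animation Blueprint for the ExcitedCyan backpack cosmetic, adding secondary motion and a timed material effect to the pack.
# AnimGraph offsets the pack bone through a bool-alpha node, then chains AnimDynamics on head and zipper bones into a RigidBody physics simulation.
# Each update, a thresholded sine of real time briefly and periodically swaps two "Hide Element" material parameters, like a blink.
# Initialization and customization-load events set Is Large Female from the pawn's body type and gender.
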AnimBlueprint /BRCosmetics/Accessories/FORT_Backpacks/Backpack_F_MED_ExcitedCyan/Meshes/F_MED_ExcitedCyan_Pack_Skeleton_AnimBlueprint
target_skeleton=&/F_MED_ExcitedCyan_Pack_Skeleton
graphs [
 AnimGraph {nodes [
   n1 = "Transform (Modify) Bone - Bone: pack"!(ComponentPose=Local To Component(LocalPose=Input Pose()), bAlphaBoolEnabled:bool, Translation="0.000,1.661,-0.000", Rotation="0,0,0", Scale) -> (Pose) {
    alphaInputType=Bool
    translationMode=BMM_Additive
    translationSpace=BCS_BoneSpace
    alphaBoolBlend {blendOption=Linear}
    boneToModify {boneName=pack}
    translation {y=1.661}
   }
   n2 = Anim Dynamics - Bone: head_kit_dyn!(ComponentPose=@n1, Alpha=1, GravityOverride:Vector="0,0,-900", ExternalForce="0,0,0") -> (Pose) {
    angularBiasOverride=1
    angularDampingOverride=1
    angularSpringConstant=10000
    bAngularSpring=true
    bOverrideAngularBias=true
    bOverrideAngularDamping=true
    bUseGravityOverride=true
    bUsePlanarLimit=false
    simulationSpace=RootRelative
    alphaBoolBlend {blendInTime=0.2, blendOption=CircularIn, blendOutTime=0.3}
    boundBone {boneName=head_kit_dyn}
    componentAppliedLinearAccClamp=<100 100 100>
    componentLinearAccScale=<0.2 0.2 0.2>
    componentLinearVelScale=<0.2 0.2 0.2>
    physicsBodyDefinitions [{
      sphereCollisionRadius=4
      boundBone {boneName=head_kit_dyn}
      boxExtents=<10 4 6>
      constraintSetup {angularLimitsMax=<10 14 2>, angularLimitsMin=<-10 -10 -2>, angularTarget=<0 0 0>}
      localJointOffset=<0 0.4 -1.2>
     }]
   }
   n3 = Anim Dynamics - Bone: zip_02_dyn!(ComponentPose=@n2, Alpha=1, GravityOverride:Vector="0,0,-300", ExternalForce="0,0,0") -> (Pose) {
    angularBiasOverride=1
    angularDampingOverride=0.3
    angularSpringConstant=1000
    bAngularSpring=true
    bOverrideAngularDamping=true
    bUseGravityOverride=true
    bUsePlanarLimit=false
    simulationSpace=RootRelative
    alphaBoolBlend {blendInTime=0.5, blendOption=Linear, blendOutTime=0.7}
    boundBone {boneName=zip_02_dyn}
    componentAppliedLinearAccClamp=<200 200 200>
    componentLinearAccScale=<0.3 0.3 0.3>
    componentLinearVelScale=<0.4 0.4 0.4>
    physicsBodyDefinitions [{
      collisionType=CustomSphere
      sphereCollisionRadius=2.5
      boundBone {boneName=zip_02_dyn}
      boxExtents=<2 2 2>
      constraintSetup {angularTargetAxis=AxisY, twistAxis=AxisY, angularLimitsMax=<2 2 0>, angularLimitsMin=<-85 0 0>, angularTarget=<0 0 0>}
      localJointOffset=<0 0 -0.3682>
     }]
   }
   n4 = RigidBody(ComponentPose=@n3) {
    bOverrideWorldGravity=true
    overridePhysicsAsset=&/F_MED_ExcitedCyan_Pack_Physics
    alphaBoolBlend {blendOption=Linear}
    componentAppliedLinearAccClamp=<9000 9000 9000>
    componentLinearAccScale=<0.4 0.4 0.4>
    componentLinearVelScale=<0.4 0.4 0.4>
    overrideWorldGravity {z=-980}
    simSpaceSettings {
     maxAngularAcceleration=20
     maxAngularVelocity=3
     maxLinearAcceleration=200
     maxLinearVelocity=250
     velocityScaleZ=0.13
     worldAlpha=1
     externalLinearDragV=<0.1 0.1>
    }
   }
   Output Pose(Result=Component To Local(ComponentPose=@n4))
  ]}
 EventGraph {nodes [
   Event BlueprintUpdateAnimation()
   n1 = GetRealTimeSeconds()
   n2 = "float > float"(A="Sin (Radians)"(A=float * float(A=@n1, B=2) {is=title K2Node_PromotableOperator}), B=0.96)
   n3 = GetPartSkeletalMeshComponent()
   > SetScalarParameterValueOnMaterials!(
    execute
    self:Mesh Component Object Reference=@n3
    ParameterName:Name=Hide Element 03
    ParameterValue:float=SelectFloat(A=1, B=0, bPickA=@n2)
   ) -> (then)
   > SetScalarParameterValueOnMaterials(self=@n3, ParameterName=Hide Element 06, ParameterValue=SelectFloat(A=0, B=1, bPickA=@n2))
   n4 = Event OnCharacterAndPickaxeCustomizationLoaded!() -> (
    OutputDelegate
    then
    CustomBaseMesh:Skeletal Mesh Component Object Reference
    CustomParts:Array of Custom Character Part Object References
    CustomMeshParts:Array of Skeletal Mesh Component Budgeted Object References
   ) {is=title K2Node_Event}
   n5 = TryGetFortPawn()
   n6 = FindBodyType!(execute=[@n4.then, @n8.then], Object:Object Reference=@n5) -> (True:Exec, False:Exec, OutBodyType:EFortCustomBodyType Enum=NONE)
   > n7 = FindGender!(execute, Object:Object Reference=@n5) -> (True:Exec, False:Exec, OutGender:EFortCustomGender Enum=Invalid)
   > Set Is Large Female(Is Large Female:bool=AND Boolean(
     A="Equal (Enum)"(A:EFortCustomGender Enum=@n7, B:EFortCustomGender Enum=Female)
     B="Equal (Enum)"(A:EFortCustomBodyType Enum=@n6, B:EFortCustomBodyType Enum=Large)
    )) -> (Output_Get:bool)
   n8 = Event BlueprintInitializeAnimation()
   TryGetPawnOwner()
  ]}
]
variables [Is Large Female {type=boolean}]
